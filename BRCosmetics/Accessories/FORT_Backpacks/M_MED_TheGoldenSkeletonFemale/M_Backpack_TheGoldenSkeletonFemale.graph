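# Translucent, unlit, two-sided VFX material for the Golden Skeleton Female backpack, flagged for Niagara mesh particles.
# SplineThicken plus a switchable, vertex-color-weighted sine wave along normals forms the world position offset.
# Emissive color samples a curve atlas using an object-centered sphere mask and panning gradients, tinted by particle color.
# Opacity blends panning and static trail textures with depth fade; SphereMaskR switches to a UV sphere mask and plain particle color.
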Material /BRCosmetics/Accessories/FORT_Backpacks/M_MED_TheGoldenSkeletonFemale/M_Backpack_TheGoldenSkeletonFemale
graph {
 properties {
  bUsedWithNiagaraMeshParticles=true
  blendMode=BLEND_Translucent
  pixelDepthOffsetMode=PDOM_Legacy
  shadingModel=MSM_Unlit
  twoSided=true
  usedShadingModels=Unlit
 }
 nodes [
  n1 = TextureCoordinate() {coordinateIndex=2}
  n2 = Add(A=Multiply(A=ComponentMask(None=@n1) {g=true}) {constB=2}, B=Time())
  n3 = TextureCoordinate() {coordinateIndex=2}
  n4 = Multiply(A=Multiply(A=Sine(None=@n2), B=VectorParameter() {defaultValue=<rgba 1 1 0 1>}.RGB), B=OneMinus(None=ComponentMask(None=@n3) {g=true}))
  n5 = VertexNormalWS()
  n6 = Multiply(A=Transform(None=Multiply(A=@n4, B=@n5)), B=VertexColor())
  n7 = ScalarParameter() {defaultValue=0.7, parameterName=Param_4}
  Add(
   A=MaterialFunctionCall(WidthBase=ScalarParameter() {defaultValue=12, parameterName=B}, WidthTip=ScalarParameter() {defaultValue=3, parameterName=T}) {materialFunction=&SplineThicken}.WorldPositionOffset
   B=LinearInterpolate(A=Multiply(A=@n6, B=@n7), B=Constant(), Alpha=ScalarParameter() {parameterName=Don't_UseWPO})
  )
  Constant() {r=1}
  n8 = TextureSample!(UVs=Panner(Coordinate=TextureCoordinate() {coordinateIndex=1}) {speedY=0.18}, Tex, Level=Constant(), Apply View MipBias) {
   automaticViewMipBias=false
   mipValueMode=TMVM_MipLevel
   texture=&/P_Backpack_GoldenSkeleton_Trail_test
  }
  n9 = LinearInterpolate(
   A=@n8.RGB
   B=TextureSample!(UVs=TextureCoordinate(), Tex, Apply View MipBias) {texture=&/P_Backpack_GoldenSkeleton_Trail_test1}.RGB
   Alpha=ScalarParameter() {parameterName=UseTrail}
  )
  n10 = MaterialFunctionCall(UVs=TextureCoordinate(), "Rotation Angle (0-1)"=Constant() {r=0.25}) {materialFunction=&CustomRotator}
  n11 = TextureSample!(UVs=@n10.Rotated Values, Tex, Level=Constant(), Apply View MipBias) {mipValueMode=TMVM_MipLevel, texture=&/P_Backpack_GoldenSkeleton}
  n12 = ScalarParameter()
  n13 = ComponentMask(None=TextureCoordinate() {coordinateIndex=1}) {r=true}
  n14 = Constant() {r=0.5}
  n15 = ParticleColor()
  n16 = ScalarParameter() {parameterName=SphereMaskR}
  LinearInterpolate(
   A=DepthFade(Opacity=LinearInterpolate(A=@n9, B=@n11.RGB, Alpha=@n12), FadeDistance=ScalarParameter() {defaultValue=30, parameterName=DepthFade})
   B=Multiply(A=SphereMask(A=@n13, B=@n14) {attenuationRadius=0.4, hardnessPercent=90}, B=@n15.A)
   Alpha=@n16
  )
  n17 = SphereMask(
   A=WorldPosition().XYZ
   B=ObjectPositionWS()
   Radius=ScalarParameter() {defaultValue=17, parameterName=Param_1}
   Hardness=ScalarParameter() {parameterName=Param_2}
  ) {attenuationRadius=10, hardnessPercent=1}
  n18 = TextureCoordinate() {uTiling=5, vTiling=0.35}
  n19 = Constant() {r=0.25}
  n20 = ScalarParameter()
  n21 = LinearInterpolate(
   A=@n18
   B=MaterialFunctionCall(UVs=@n18, "Rotation Angle (0-1)"=@n19) {materialFunction=&CustomRotator}.Rotated Values
   Alpha=@n20
  )
  n22 = TextureCoordinate() {uTiling=2, vTiling=0.7}
  n23 = LinearInterpolate(
   A=@n22
   B=MaterialFunctionCall(UVs=@n22, "Rotation Angle (0-1)"=@n19) {materialFunction=&CustomRotator}.Rotated Values
   Alpha=@n20
  )
  n24 = Subtract(
   A=TextureSample!(UVs=Panner(Coordinate=@n21) {speedX=-0.1}, Tex, Apply View MipBias) {texture=/Game/Maps/Textures/RadialGradient2}.R
   B=TextureSample(UVs=Panner(Coordinate=@n23) {speedX=0.04}) {texture=&RadialGradient2}.R
  )
  n25 = Multiply(A=Divide(A=Add(A=@n24)), B=@n17)
  n26 = TextureCoordinate()
  n27 = Multiply(A=ComponentMask(None=@n26) {r=true}, B=OneMinus(None=ComponentMask(None=@n26) {r=true}))
  n28 = Power(Base=Divide(A=Add(A=@n27)), Exp=ScalarParameter() {defaultValue=-1, parameterName=Param_6})
  n29 = CurveAtlasRowParameter(CurveTime=LinearInterpolate(A=@n17, B=Multiply(A=@n25, B=@n28), Alpha=ScalarParameter())) {
   atlas=/Game/Athena/LinearColorCurves/VFX_CurveAtlas
   curve=/Game/Accessories/FORT_Backpacks/M_MED_TheGoldenSkeletonFemale/Curve_TheGoldenSkeletonFemale
   defaultValue=18
   group=Curve
   parameterName=Curve
  }
  n30 = Desaturation(
   None=Multiply(A=@n29, B=ScalarParameter() {defaultValue=1.4, parameterName=Param_5})
   Fraction=ScalarParameter() {defaultValue=-0.5, parameterName=Param_3}
  )
  MaterialFunctionCall(Emissive=LinearInterpolate(A=Multiply(A=@n30, B=@n15.RGB), B=@n15.RGB, Alpha=@n16)) {materialFunction=&MF_FX_EmissiveAdjustments}
  Panner(Coordinate=@n10.Rotated Values)
  TextureCoordinate()
  TextureObjectParameter() {parameterName=Mask, texture=&P_Backpack_GoldenSkeleton_Trail_test}
 ]
}
